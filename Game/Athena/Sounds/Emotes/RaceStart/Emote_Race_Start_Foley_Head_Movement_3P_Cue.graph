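# Third-person sound cue for head-movement foley in the Race Start emote.
# A random node picks one of two equally weighted foot-scuff waves borrowed from the LockMeUp emote, then a modulator plays it quietly with randomized lower pitch and volume.
# It uses the emote music sound class and concurrency set, keeps playing when silent, and sends to the music audio-analysis submix.
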SoundCue /Game/Athena/Sounds/Emotes/RaceStart/Emote_Race_Start_Foley_Head_Movement_3P_Cue
root=Modulator_0
nodes [
 {id=Modulator_0, children [Random_1], properties {pitchMax=1, pitchMin=0.8, volumeMax=0.25, volumeMin=0.1}}
 {id=Random_1, children [WavePlayer_4,WavePlayer_5], properties {weights [1,1]}}
 {id=WavePlayer_4, properties {soundWaveAssetPtr=/Game/Athena/Sounds/Emotes/LockMeUp/Emote_LockMeUp_Foley_Foot_Scuff_Start_01}}
 {id=WavePlayer_5, properties {soundWaveAssetPtr=/Game/Athena/Sounds/Emotes/LockMeUp/Emote_LockMeUp_Foley_Foot_Scuff_Start_02}}
]
properties {
 maxDistance=2097152
 soundClassObject=&Fort_Music_Emote
 virtualizationMode=PlayWhenSilent
 volumeMultiplier=0.5
 concurrencyOverrides {maxCount=3, resolutionRule=StopOldest}
 concurrencySet [/Game/Sounds/Concurrency_Rules/Athena_EmoteMusic_Conc]
 soundSubmixSends [{minSendLevel=0, soundSubmix=&Fortnite_AudioAnalysis_Music}]
}
